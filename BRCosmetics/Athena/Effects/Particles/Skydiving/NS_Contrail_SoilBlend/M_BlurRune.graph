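# Translucent material for Niagara ribbons and sprites in the SoilBlend skydiving contrail effect.
# The Texture parameter's red channel feeds all three inputs of a Spin-driven multi-lerp, then is tinted by particle color through an emissive-adjustment function.
# Particle alpha times that mask, divided by InvertDIstanceRamp, forms opacity, which also blends the depth-fade distance between DF1 and DF2.
# A shading-path switch skips depth fade on mobile; BaseColor is unconnected.
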Material /BRCosmetics/Athena/Effects/Particles/Skydiving/NS_Contrail_SoilBlend/M_BlurRune
graph {
 properties {
  bAutomaticallySetUsageInEditor=false
  bUsedWithNiagaraRibbons=true
  bUsedWithNiagaraSprites=true
  blendMode=BLEND_Translucent
  pixelDepthOffsetMode=PDOM_Legacy
  usedShadingModels=Default Lit
 }
 nodes [
  n1 = ParticleColor()
  n2 = TextureSampleParameter2D() {parameterName=Texture, texture=&/T_SoilBlend_01}
  n3 = DynamicParameter() {paramNames [Spin,Param2,Param3,Param4]}
  n4 = MaterialFunctionCall!(1=@n2.R, 2=@n2.R, 3=@n2.R, 4, A=Divide(A=Add(A=@n3.Spin))) {materialFunction=/Engine/Functions/Engine_MaterialFunctions02/Lerp_Multiple_Float}
  MaterialFunctionCall(Emissive=Multiply(A=@n1.RGB, B=@n4.Lerp 3 Inputs)) {materialFunction=&MF_FX_EmissiveAdjustments}
  n5 = Multiply(A=@n1.A, B=Divide(A=@n4.Lerp 3 Inputs, B=ScalarParameter() {defaultValue=2, parameterName=InvertDIstanceRamp}))
  n6 = DepthFade(
   Opacity=@n5
   FadeDistance=LinearInterpolate(A=ScalarParameter() {defaultValue=50, parameterName=DF1}, B=ScalarParameter() {parameterName=DF2}, Alpha=@n5)
  )
  ShadingPathSwitch(Default=@n6, Deferred=@n6, Mobile=@n5) {desc=Migrated from FeatureLevelSwitch}
  VectorParameter() {parameterName=BaseColor, defaultValue=<rgba 0.0294 0.217 0.0402 1>}
 ]
}
