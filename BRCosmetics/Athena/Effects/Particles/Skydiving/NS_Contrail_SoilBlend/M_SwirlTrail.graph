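# Translucent two-sided material for swirl trail mesh particles in the SoilBlend skydiving contrail Niagara effect.
# A panning swirl texture, with a wind-normal distortion branch, is eroded by an alpha curve along U and remapped through a tint curve atlas row for emissive color and depth-faded opacity.
# Vertex offset math combines SplineThicken with particle-relative pull, sine waves and camera-facing ribbon widening, partly driven by Niagara dynamic parameters.
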Material /BRCosmetics/Athena/Effects/Particles/Skydiving/NS_Contrail_SoilBlend/M_SwirlTrail
graph {
 properties {
  bAutomaticallySetUsageInEditor=false
  bUsedWithNiagaraMeshParticles=true
  blendMode=BLEND_Translucent
  twoSided=true
  usedShadingModels=Default Lit
 }
 nodes [
  n1 = ScalarParameter() {defaultValue=15, parameterName=SplineThickenBase}
  n2 = DynamicParameter() {parameterIndex=1, defaultValue=<rgba 1 1 1 0>, paramNames [WPO, Speed, WPO Disturbance, U offset]}
  n3 = ComponentMask(None=TextureCoordinate()) {r=true}
  n4 = ScalarParameter() {parameterName=SplineThicken Edge Strenght}
  n5 = Saturate(None=Multiply(A=LinearInterpolate(A=@n3, Alpha=@n4), B=ScalarParameter() {defaultValue=1, parameterName=SplineTicken Edge Contrast}) {constB=5})
  n6 = ScalarParameter() {defaultValue=150, parameterName=SplineThickenTip}
  n7 = MaterialFunctionCall(
   WidthBase=Multiply(A=Multiply(A=@n1, B=@n2.WPO), B=@n5)
   WidthTip=Multiply(A=Multiply(A=@n6, B=@n2.WPO), B=@n5)
   Expand U or V UV Channel=StaticBool()
  ) {materialFunction=&SplineThicken}
  n8 = DynamicParameter() {parameterIndex=2, defaultValue=<rgba 1 1 1 0>, paramNames [X,Y,Z,WPOIntensity]}
  n9 = Clamp(
   None=Subtract(
    A=MaterialFunctionCall(X=@n8.X, Y=@n8.X, Z=@n8.X) {materialFunction=&MakeFloat3}.Result
    B=ParticlePositionWS()
   )
   Min=Constant() {r=-30}
   Max=Constant() {r=30}
  )
  n10 = TextureCoordinate()
  n11 = Multiply(A=@n8.WPOIntensity, B=Power(Base=ComponentMask(None=@n10) {r=true}, Exp=ScalarParameter() {defaultValue=2, parameterName=WPOBlendPower}))
  Add(
   A=@n7.WorldPositionOffset
   B=LinearInterpolate(B=LinearInterpolate(B=@n9, Alpha=@n11), Alpha=ScalarParameter() {parameterName=Use Interpolate WPO})
  )
  n12 = DynamicParameter() {parameterIndex=1, defaultValue=<rgba 1 1 1 0>, paramNames [WPO, Speed, WPO Disturbance, U offset]}
  n13 = Constant()
  n14 = ComponentMask(None=Add(A=TextureCoordinate() {uTiling=2.2}, B=AppendVector(A=@n12.U offset, B=@n13))) {r=true}
  n15 = Multiply(A=Time()) {constB=-1.3}
  n16 = Multiply(A=Sine(None=Add(A=@n14, B=@n15)) {period=1.2}, B=VectorParameter() {parameterName=WPO Direction, defaultValue=<rgba 0 0 1 1>}.RGB)
  n17 = LinearInterpolate(
   A=Transform(None=@n16) {transformSourceType=TRANSFORMSOURCE_World, transformType=TRANSFORM_Camera}
   B=Transform(None=@n16) {transformSourceType=TRANSFORMSOURCE_Local}
   Alpha=ScalarParameter() {parameterName=Change Transform}
  )
  n18 = DynamicParameter() {parameterIndex=1, defaultValue=<rgba 1 1 1 0>, paramNames [WPO, Speed, WPO Disturbance, U offset]}
  n19 = TextureCoordinate()
  n20 = Saturate(None=Multiply(A=ComponentMask(None=@n19) {r=true}) {constB=3})
  Add(A=@n7.WorldPositionOffset, B=Multiply(A=Multiply(A=@n17, B=@n18.WPO Disturbance), B=LinearInterpolate(Alpha=@n20))) {constB=0}
  Constant()
  n21 = VectorParameter() {parameterName=Texture_SV, defaultValue=<rgba 2.3585 1 1.0652 0>}
  n22 = DynamicParameter() {paramNames [Time,Param2,Param3,Param4]}
  n23 = TextureSample(UVs=Add(A=MaterialFunctionCall!(InUV, Texture, InRGB=@n21.RGB, InA=@n21.RGB, InTime=@n22.Time) {materialFunction=/Game/Effects/Fort_Effects/Materials/_Functions/MF_TexturePanner}.UV)) {texture=&/T_Trail_Swirl}
  n24 = CurveAtlasRowParameter(CurveTime=ComponentMask(None=TextureCoordinate()) {r=true}) {
   atlas=&/CurveAtlas_SoilBlend
   curve=&/Curve_Alpha
   parameterName=Alpha
  }
  n25 = CurveAtlasRowParameter(CurveTime=Subtract(A=ChannelMaskParameter(None=@n23.RGBA) {parameterName=TextureMask}.R, B=@n24.R)) {
   atlas=&CurveAtlas_SoilBlend
   curve=&/Curve_Tint
   defaultValue=1
   parameterName=Tint
  }
  DepthFade(Opacity=@n25.A)
  MaterialFunctionCall(Emissive=@n25) {materialFunction=&MF_FX_EmissiveAdjustments}
  n26 = TextureCoordinate()
  n27 = Subtract(A=WorldPosition().XYZ, B=CameraPositionWS())
  n28 = CrossProduct(A=@n27, B=Transform(None=Constant3Vector() {constant=<rgba 1 0 0 0>}))
  n29 = TextureCoordinate()
  n30 = LinearInterpolate(
   A=ScalarParameter() {defaultValue=35.3707, parameterName=RibbonTopWidth}
   B=ScalarParameter() {defaultValue=94.104, parameterName=RibbonEndWidth}
   Alpha=ComponentMask(None=@n29) {r=true}
  )
  n31 = ScalarParameter() {defaultValue=5}
  n32 = ScalarParameter() {defaultValue=2, parameterName=Param_1}
  n33 = Subtract(A=WorldPosition().XYZ, B=CameraPositionWS())
  n34 = Abs(None=ComponentMask(None=Normalize(VectorInput=@n33)) {b=true})
  n35 = Saturate(None=LinearInterpolate(A=LinearInterpolate(A=@n31, B=@n32, Alpha=@n34), Alpha=ComponentMask(None=@n29) {r=true}) {constB=0})
  Multiply(A=Subtract(A=ComponentMask(None=@n26) {g=true}) {constB=0.5}, B=Multiply(A=Normalize(VectorInput=@n28), B=Multiply(A=@n30, B=@n35)))
  n36 = ComponentMask(None=TextureCoordinate()) {g=true}
  n37 = VectorParameter() {parameterName=Distort_SV, defaultValue=<rgba 1 1 0.2213 0>}
  n38 = DynamicParameter() {paramNames [Time,Param2,Param3,Param4]}
  n39 = Multiply(
   A=Abs(None=Subtract(A=@n36) {constB=0.5})
   B=TextureSample(UVs=MaterialFunctionCall(InRGB=@n37.RGB, InA=@n37.RGB, InTime=@n38.Time) {materialFunction=&MF_TexturePanner}.UV) {samplerType=SAMPLERTYPE_Normal, texture=/Game/Effects/Fort_Effects/Textures/Tile/T_Wind_N}.RGB
  )
  n40 = ScalarParameter() {defaultValue=0.4, parameterName=Distort}
  NamedRerouteDeclaration(None=ComponentMask(None=Multiply(A=@n39, B=@n40)) {g=true, r=true}) {name=Distort, nodeColor=<rgba 0.875 0 1 1>}
  NamedRerouteUsage()
  Normalize(VectorInput=@n27)
  Time()
  VectorParameter() {parameterName=BaseColor, defaultValue=<rgba 0.032 0.1389 0.0643 1>}
 ]
}
