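# Green skydiving contrail cosmetic (SoilBlend), a Niagara system attached to the character's skeletal mesh.
# Every emitter re-samples one of two sockets each frame, chosen by particle ID parity; all but Glow use the hand mesh interface.
# Glow pulses sprites and lights, Dots swirl in vortices, HandDetail adds animated flipbook puffs, and MeshRibbon layers four mesh renderers with time-driven material parameters.
# Rune layers spinning sprites through per-particle arrays, with spin speed driving color blending.
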NiagaraSystem /BRCosmetics/Athena/Effects/Particles/Skydiving/NS_Contrail_SoilBlend/NS_SoilBlend
dependencies {
 usedDataInterfaces [
  &NiagaraDataInterfaceCurve
  &NiagaraDataInterfaceColorCurve
  &NiagaraDataInterfaceVector2DCurve
  &NiagaraDataInterfaceArrayFloat
  &NiagaraDataInterfaceSkeletalMesh
 ]
 usedDynamicInputs []
 usedModules [
  &self:SystemSpawnScript.SystemScriptSource.SystemScriptGraph.NiagaraNodeAssignment_0.SetVariables_#1
  &self:SystemSpawnScript.SystemScriptSource.SystemScriptGraph.NiagaraNodeAssignment_1.SetVariables_#2
  &SystemState
 ]
}
emitters {
 Dots {
  emitterUpdateScript [
   &EmitterState(Life Cycle Mode=::NewEnumerator1, Loop Duration=0.02, UseLoopDelay=false)
   &SpawnBurst_Instantaneous(Loop Count Limit=0, Spawn Count=4, Use Spawn Probability=false, UseLoopCountLimit=false) {enabled=false}
   &SpawnPerFrame(Use Spawn Probability=false) {enabled=false}
   &SpawnRate(SpawnRate=12.6762, Use Spawn Probability=false)
   /Game/Effects/Niagara/Modules/Spawn/FN_SkydiveSpawnRate(Global Spawnrate=18) {enabled=false}
  ]
  particleSpawnScript [
   /SetVariables(Particles.SocketIndex=HlslExpression() {hlslExpression="fmod(Particles.UniqueID,2)"}) {isSetParameters=true, module=SetVariables_#3}
   /Game/Effects/Niagara/Modules/Groups/FN_SetSpawnGroup_v1/FN_SetSpawnGroup(Override SpawnGroup=true, SpawnGroup:NiagaraInt32=&SpawnGroup()) {enabled=false}
   &InitializeParticle(
    Color Mode=::NewEnumerator0
    Lifetime=&RandomRangeFloat(Minimum=1, Maximum=5)
    Position Mode=::NewEnumerator1
    Sprite Size Mode=::NewEnumerator3
    Uniform Sprite Size=&RandomRangeFloat(Minimum=1, Maximum=2.7228)
    UsePositionOffset=false
    Write Material Random=true
   )
   /SetVariables(Particles.RibbonID=/Niagara/DynamicInputs/NiagaraID/MakeNiagaraIDFromInt(Index:NiagaraInt32=$Particles.SocketIndex, Acquire Tag:NiagaraInt32=$Particles.SocketIndex)) {enabled=false, isSetParameters=true, module=SetVariables_#4}
   &InheritVelocity(
    Inherited Velocity Amount Scale=&VectorFromFloat(Value=-0.05)
    Inherited Velocity Speed Limit=10000
    Speed Limit=true
    UseSpeedThreshold=false
   ) {enabled=false}
   &AddVelocity(Velocity=<0 0 100>) {enabled=false}
  ]
  particleUpdateScript [
   &SampleSkeletalMesh(
    AdditionalYawPitchRoll=false
    Bone / Socket Index:NiagaraInt32=$Particles.SocketIndex
    Bone Sampling Mode:ENiagaraBoneSamplingMode=::NewEnumerator4
    Position Offset Coordinate Space:ENiagaraCoordinateSpace=::World
    Skeletal Mesh:NiagaraDataInterfaceSkeletalMesh=$System.Skeletal Mesh Hand
   )
   /SetVariables(Particles.Position=$StackContext.SampleSkeletalMesh.SampledPosition) {isSetParameters=true, module=SetVariables_#5}
   &self:OffsetSprites(
    Add Vortex=true
    BlendPercentage:NiagaraFloat=&FloatFromCurve(
     FloatCurve=DataInterface() {propertyValues {Curve {keys [
         {arriveTangent=0.22, interpMode=RCIM_Cubic, leaveTangent=0.22, tangentMode=RCTM_User, time=0, value=0}
         {arriveTangent=1.6752, interpMode=RCIM_Cubic, leaveTangent=1.6752, tangentMode=RCTM_User, time=1, value=0.875}
        ]}}}
     Scale Curve=3
    )
    MaxiumVelAdded=0
    OffsetPosition:NiagaraPosition=<0 0 200>
    SampledPosition:NiagaraPosition=$StackContext.SampleSkeletalMesh.SampledPosition
    VelocityAddUp=-0.01
    Vortex Distance:NiagaraFloat=&Multiply_Float(
     A=&RandomRangeFloat(Minimum=3, Maximum=15)
     B=&FloatFromCurve(FloatCurve=DataInterface() {propertyValues {Curve {keys [<at 0 0.5 RCIM_Cubic>,<at 1 2 RCIM_Cubic>]}}}, Scale Curve=3)
    )
    VortexSpeed:NiagaraFloat=&Add_Float(
     A=&RandomRangeFloat(Minimum=0.1, Maximum=0.2)
     B=&FloatFromCurve(
      FloatCurve=DataInterface() {propertyValues {Curve {keys [{arriveTangent=6.97, interpMode=RCIM_Cubic, leaveTangent=6.97, tangentMode=RCTM_User, time=0, value=1}, <at 1 3.75 RCIM_Cubic>]}}}
      Scale Curve=&RandomRangeFloat(Minimum=0, Maximum=0.3973)
     )
    )
   ) {module=OffsetSprites}
   &ShapeLocation(MaskSpawnGroup=false, Override Seed=false, Shape Origin=$Particles.Position, Sphere Radius=20, UseOwnerScale=true) {enabled=false}
   &ParticleState()
   &CameraOffset(Camera Offset Amount=&FloatFromCurve(FloatCurve=DataInterface() {propertyValues {Curve {keys [<at 0 50>,<at 1 -50>]}}}))
   &ScaleSpriteSize(Scale Factor=&Vector2DFromFloat(Value=&FloatFromCurve(
      FloatCurve=DataInterface() {propertyValues {Curve {keys [{arriveTangent=7.4819, interpMode=RCIM_Cubic, leaveTangent=7.4819, tangentMode=RCTM_User, time=0, value=0}, <at 0.0979 0.9367>, <at 1 0>]}}}
      Scale Curve=&RandomRangeFloat(Minimum=1, Maximum=1.2, Evaluation Type=::NewEnumerator1)
     )))
   &ScaleSpriteSize(
    Initial Sprite Size:Vector2f=$Particles.SpriteSize
    Scale Sprite Size Mode=::NewEnumerator0
    Uniform Scale Factor=&RandomRangeFloat(Minimum=0.8, Maximum=1.2, Evaluation Type=::NewEnumerator1)
   ) {module=ScaleSpriteSize001}
   &ScaleColor(
    Linear Color Curve=DataInterface() {propertyValues {
      AlphaCurve {keys [<at 0 1>,<at 0.9954 1>]}
      BlueCurve {keys [<at 0 0>,<at 0.0562 0>,<at 0.1886 0>,<at 0.2274 0>,<at 0.2548 0>,<at 0.7512 0>,<at 0.782 0>,<at 0.8459 0>,<at 1 0>]}
      GreenCurve {keys [<at 0 50>,<at 0.0562 2>,<at 0.1886 2>,<at 0.2274 50>,<at 0.2548 2>,<at 0.7512 2>,<at 0.782 50>,<at 0.8459 2>,<at 1 2>]}
      RedCurve {keys [<at 0 0>,<at 0.0562 0>,<at 0.1886 0>,<at 0.2274 0>,<at 0.2548 0>,<at 0.7512 0>,<at 0.782 0>,<at 0.8459 0>,<at 1 0>]}
     }}
    Scale Mode=::NewEnumerator2
   )
  ]
  rendererClasses [&NiagaraSpriteRendererProperties]
  renderers [{rendererClass=&NiagaraSpriteRendererProperties}]
 }
 Glow {
  emitterUpdateScript [
   &EmitterState(
    Inactive Response=::NewEnumerator2
    Life Cycle Mode=::NewEnumerator1
    Loop Behavior=::NewEnumerator1
    Loop Duration Mode=::NewEnumerator1
    Scalability Mode=::NewEnumerator1
    UseLoopDelay=false
   )
   &SpawnBurst_Instantaneous(Loop Count Limit=0, Spawn Count=2, Use Spawn Probability=false, UseLoopCountLimit=false)
   &SpawnPerFrame(Use Spawn Probability=false) {enabled=false}
   &SpawnRate(SpawnRate=20, Use Spawn Probability=false) {enabled=false}
   &FN_SkydiveSpawnRate(Global Spawnrate=18) {enabled=false}
  ]
  particleSpawnScript [
   /SetVariables(Particles.SocketIndex=HlslExpression() {hlslExpression="fmod(Particles.UniqueID,2)"}) {isSetParameters=true, module=SetVariables_#3}
   &InitializeParticle(
    Color Mode=::NewEnumerator0
    Lifetime=&RandomRangeFloat(Minimum=1, Maximum=2)
    Position Mode=::NewEnumerator1
    Sprite Size=<498.5609 114.2901>
    Sprite Size Mode=::NewEnumerator4
    UsePositionOffset=false
    Write Material Random=true
   )
   &DynamicMaterialParameters(
    Index 0 Param 1=0.2
    Index 0 Param 2=1
    Module.Param0WriteEnabled=true
    Module.Param1WriteEnabled=true
    Module.Param2WriteEnabled=true
    Module.Param3WriteEnabled=true
   )
  ]
  particleUpdateScript [
   &SampleSkeletalMesh(
    AdditionalYawPitchRoll=false
    Bone / Socket Index:NiagaraInt32=$Particles.SocketIndex
    Bone Sampling Mode:ENiagaraBoneSamplingMode=::NewEnumerator4
    Position Offset Coordinate Space:ENiagaraCoordinateSpace=::World
    Sampled Position Offset:Vector3f=<0 12 0>
    Skeletal Mesh:NiagaraDataInterfaceSkeletalMesh=$System.Skeletal Mesh
   )
   /SetVariables(Particles.Position=&AddVectorToPosition(Position=<0 0 5>, Vector=Linked() {linkedVariable {name=StackContext.SampleSkeletalMesh.SampledPosition, type=NiagaraPosition}})) {isSetParameters=true, module=SetVariables_#5}
   &self:OffsetSprites(
    BlendPercentage:NiagaraFloat=&FloatFromCurve(FloatCurve=DataInterface() {propertyValues {Curve {keys [
         {arriveTangent=-0.0296, interpMode=RCIM_Cubic, leaveTangent=-0.0296, tangentMode=RCTM_User, time=0, value=0}
         {arriveTangent=2.0167, interpMode=RCIM_Cubic, leaveTangent=2.0167, tangentMode=RCTM_User, time=1, value=1}
        ]}}})
    MaxiumVelAdded=0
    OffsetPosition:NiagaraPosition=<0 0 200>
    SampledPosition:NiagaraPosition=$StackContext.SampleSkeletalMesh.SampledPosition
    VelocityAddUp=-0.01
   ) {enabled=false, module=OffsetSprites}
   &SolveForcesAndVelocity(Clamp Velocity=false, Limit Acceleration=false) {enabled=false}
   &ParticleState(
    Kill Particles When Lifetime Has Elapsed=false
    Let Infinitely Lived Particles Die When Emitter Deactivates=true
    Loop Particles Lifetime=true
   )
   &CameraOffset(Camera Offset Amount=&FloatFromCurve(FloatCurve=DataInterface() {propertyValues {Curve {keys [<at 0 50>,<at 0.5 -100>,<at 1 50>]}}}))
   &self:ColorSet(
    Linear Color:LinearColor=<rgba 0 1 0.1294 1>
    Min=0.5
    Random:NiagaraFloat=&RandomRangeFloat(Minimum=0, Maximum=0.1, Evaluation Type=::NewEnumerator1)
    ScaleCurve:NiagaraFloat=&FloatFromCurve(
     FloatCurve=DataInterface() {propertyValues {Curve {keys [<at 0 0>,<at 0.1 0.5>,<at 0.2 0>,<at 0.4 0>,<at 0.5 0.75>,<at 0.6 0>,<at 0.7 0>,<at 0.8 0.25>,<at 0.9 0>,<at 1 0>]}}}
     Scale Curve=1.5
    )
   ) {module=ColorSet}
   &FN_Light_Attributes(
    Color=$Particles.Color
    Create Color Attribute=true
    Create Exponent Attribute=true
    Create Radius Attribute=true
    Enable Particle Light Enable=true
    Enable Visibility Tag=true
    Exponent=2
    Light Enabled=true
    Radius=50.5276
    Scale Color=20
    VisibilityTag=1
   )
  ]
  rendererClasses [&NiagaraSpriteRendererProperties, &NiagaraLightRendererProperties]
  renderers [
   {rendererClass=&NiagaraSpriteRendererProperties}
   {rendererClass=&NiagaraLightRendererProperties, rendererIndex=1}
  ]
 }
 HandDetail {
  emitterUpdateScript [
   &EmitterState(Life Cycle Mode=::NewEnumerator1, Loop Duration=0.02, UseLoopDelay=false)
   &SpawnBurst_Instantaneous(Loop Count Limit=0, Spawn Count=4, Use Spawn Probability=false, UseLoopCountLimit=false) {enabled=false}
   &SpawnPerFrame(Use Spawn Probability=false) {enabled=false}
   &SpawnRate(SpawnRate=31.5018, Use Spawn Probability=false)
   &FN_SkydiveSpawnRate(Global Spawnrate=18) {enabled=false}
  ]
  particleSpawnScript [
   /SetVariables(Particles.SocketIndex=HlslExpression() {hlslExpression="fmod(Particles.UniqueID,2)"}) {isSetParameters=true, module=SetVariables_#3}
   &FN_SetSpawnGroup(Override SpawnGroup=true, SpawnGroup:NiagaraInt32=&SpawnGroup()) {enabled=false}
   &InitializeParticle(
    Color Mode=::NewEnumerator0
    Lifetime=&RandomRangeFloat(Minimum=0.5, Maximum=0.8)
    Position Mode=::NewEnumerator1
    Sprite Size=<40.1585 154.6742>
    Sprite Size Mode=::NewEnumerator4
    UsePositionOffset=false
    Write Material Random=true
   )
   /SetVariables(Particles.RibbonID=&MakeNiagaraIDFromInt(Index:NiagaraInt32=$Particles.SocketIndex, Acquire Tag:NiagaraInt32=$Particles.SocketIndex)) {enabled=false, isSetParameters=true, module=SetVariables_#4}
   &InheritVelocity(
    Inherited Velocity Amount Scale=&VectorFromFloat(Value=1)
    Inherited Velocity Speed Limit=10000
    Speed Limit=true
    UseSpeedThreshold=false
   ) {enabled=false}
   &AddVelocity(Velocity=&RandomVector(Vector Scale=173.6387))
  ]
  particleUpdateScript [
   &SampleSkeletalMesh(
    AdditionalYawPitchRoll=false
    Bone / Socket Index:NiagaraInt32=$Particles.SocketIndex
    Bone Sampling Mode:ENiagaraBoneSamplingMode=::NewEnumerator4
    Position Offset Coordinate Space:ENiagaraCoordinateSpace=::World
    Sampled Position Offset:Vector3f=<0 18.4036 0>
    Skeletal Mesh:NiagaraDataInterfaceSkeletalMesh=$System.Skeletal Mesh Hand
   )
   /SetVariables(Particles.Position=$StackContext.SampleSkeletalMesh.SampledPosition) {isSetParameters=true, module=SetVariables_#5}
   &ShapeLocation(MaskSpawnGroup=false, Override Seed=false, Shape Origin=$Particles.Position, Sphere Radius=13.5865, UseOwnerScale=true)
   &self:OffsetSprites(
    BlendPercentage:NiagaraFloat=&FloatFromCurve(
     FloatCurve=DataInterface() {propertyValues {Curve {keys [
         {arriveTangent=-0.0296, interpMode=RCIM_Cubic, leaveTangent=-0.0296, tangentMode=RCTM_User, time=0, value=0}
         {arriveTangent=2.0167, interpMode=RCIM_Cubic, leaveTangent=2.0167, tangentMode=RCTM_User, time=1, value=1}
        ]}}}
     Scale Curve=1.5
    )
    MaxiumVelAdded=0
    OffsetPosition:NiagaraPosition=<0 0 200>
    SampledPosition:NiagaraPosition=$StackContext.SampleSkeletalMesh.SampledPosition
    VelocityAddUp=-0.01
   ) {enabled=false, module=OffsetSprites}
   &Drag(Drag=3.6321, Use Linear Drag=true, Use Rotational Drag=true) {enabled=false}
   &SolveForcesAndVelocity(Clamp Velocity=false, Limit Acceleration=false)
   &ParticleState()
   &CameraOffset(Camera Offset Amount=&FloatFromCurve(FloatCurve=DataInterface() {propertyValues {Curve {keys [<at 0 50>,<at 1 -50>]}}}))
   &ScaleSpriteSize(Scale Factor=&Vector2DFromCurve(
     Vector2Curve=DataInterface() {propertyValues {
       XCurve {keys [{arriveTangent=1.7106, interpMode=RCIM_Cubic, leaveTangent=1.7106, tangentMode=RCTM_User, time=0, value=-0.025}, <at 1 1 RCIM_Cubic>]}
       YCurve {keys [{arriveTangent=1.7106, interpMode=RCIM_Cubic, leaveTangent=1.7106, tangentMode=RCTM_User, time=0, value=-0.025}, <at 1 1 RCIM_Cubic>]}
      }}
     ScaleCurve=<1.6991 1.5591>
    ))
   &ScaleColor(
    Linear Color Curve=DataInterface() {propertyValues {
      AlphaCurve {keys [<at 0.0209 0>,<at 0.2815 0.8>,<at 0.7561 0.6>,<at 1 0>]}
      BlueCurve {keys [<at 0.051 0.0255>,<at 0.2448 0.1224>,<at 1 0.0949>]}
      GreenCurve {keys [<at 0.051 1>,<at 0.2448 1.5>,<at 1 0.0949>]}
      RedCurve {keys [<at 0.051 0.0312>,<at 0.2448 0.1305>,<at 1 0.0949>]}
     }}
    Scale Mode=::NewEnumerator2
   )
   &self:GroupSwitch(ColorOvrride:LinearColor=<rgba 0.0369 0.1788 0.0865 0.637>, SizeOvrride:Vector2f=<53.4972 67.6002>) {module=GroupSwitch}
   /SetVariables(Particles.SpriteRotation:NiagaraFloat=&Add_Float(
     A=&RandomRangeFloat(Minimum=0, Maximum=360)
     B=&FloatFromCurve(
      FloatCurve=DataInterface() {propertyValues {Curve {keys [{arriveTangent=-0.9167, interpMode=RCIM_Cubic, time=0, value=0.5}, {interpMode=RCIM_Cubic, tangentMode=RCTM_User, time=0.6, value=-0.5}]}}}
      Scale Curve=360
     )
    )) {isSetParameters=true}
   /SetVariables(Particles.SubImageIndex=&Add_Float(
     A=&RandomRangeFloat(Minimum=10, Maximum=30)
     B=&FloatFromCurve(FloatCurve=DataInterface() {propertyValues {Curve {keys [
          {arriveTangent=15.4725, interpMode=RCIM_Cubic, leaveTangent=15.4725, tangentMode=RCTM_User, time=0, value=0}
          {arriveTangent=31.7695, interpMode=RCIM_Cubic, leaveTangent=31.7695, tangentMode=RCTM_User, time=1, value=35}
         ]}}})
    )) {isSetParameters=true}
  ]
  rendererClasses [&NiagaraSpriteRendererProperties]
  renderers [{rendererClass=&NiagaraSpriteRendererProperties}]
 }
 MeshRibbon {
  emitterUpdateScript [
   &EmitterState(Inactive Response=::NewEnumerator2, Life Cycle Mode=::NewEnumerator1, Loop Behavior=::NewEnumerator1, Loop Duration=1, UseLoopDelay=false)
   &SpawnBurst_Instantaneous(Spawn Count=2, Use Spawn Probability=false, UseLoopCountLimit=false)
  ]
  particleSpawnScript [
   /SetVariables(Particles.SocketIndex=HlslExpression() {hlslExpression="fmod(Particles.UniqueID,2)"}) {isSetParameters=true, module=SetVariables_#3}
   &InitializeParticle(Color Mode=::NewEnumerator0, Lifetime=5, UsePositionOffset=false, Write Material Random=true)
   /SetVariables(
    Particles.MeshOrientation=&MakeQuatFromAxisAngle(Axis=<0 0 1>, AngleInDegrees=&RandomRangeFloat(Minimum=0, Maximum=350))
    Particles.MeshRot2:Quat4f=&MakeQuatFromAxisAngle(Axis=<0 0 1>, AngleInDegrees=&RandomRangeFloat(Minimum=0, Maximum=350))
    Particles.MeshRot3:Quat4f=&MakeQuatFromAxisAngle(Axis=<0 0 1>, AngleInDegrees=&RandomRangeFloat(Minimum=0, Maximum=350))
    Particles.MeshRot4:Quat4f=&MakeQuatFromAxisAngle(Axis=<0 0 1>, AngleInDegrees=&RandomRangeFloat(Minimum=0, Maximum=350))
   ) {enabled=false, isSetParameters=true}
   /SetVariables(
    Particles.Scale=<4 4 7.9948>
    Particles.Scale2:Vector3f=<6 6 8.6384>
    Particles.Scale3:Vector3f=<6 6 3.3443>
    Particles.Scale4:Vector3f=<0.01 0.01 9.8078>
   ) {isSetParameters=true}
  ]
  particleUpdateScript [
   &SampleSkeletalMesh(
    AdditionalYawPitchRoll=false
    Bone / Socket Index:NiagaraInt32=$Particles.SocketIndex
    Bone Sampling Mode:ENiagaraBoneSamplingMode=::NewEnumerator4
    Position Offset Coordinate Space:ENiagaraCoordinateSpace=::Local
    Sampled Position Offset:Vector3f=<0 15 -14.0219>
    Skeletal Mesh:NiagaraDataInterfaceSkeletalMesh=$System.Skeletal Mesh Hand
   )
   /SetVariables(Particles.Position=$StackContext.SampleSkeletalMesh.SampledPosition) {isSetParameters=true, module=SetVariables_#5}
   &ParticleState(Kill Particles When Lifetime Has Elapsed=false)
   &UpdateMeshOrientation(Rotation Vector=<0 0 1>) {enabled=false}
   &self:AdditionalRotation(SpeedLag=1) {enabled=false, module=AdditionalRotation}
   &Timeline(Timeline End=10000000) {enabled=false}
   &DynamicMaterialParameters(
    Index 0 Param 1=&Add_Float(
     A=&Multiply_Float(A=&ReturnNormalizedExecIndex(), B=0.5)
     B=&Multiply_Float(A=$Engine.Time, B=1.25)
    )
    Index 1 Param 1=2.0732
    Index 1 Param 2=$Particles.Timeline.Time
    Module.Index0Param0Write=true
    Module.Index0Param1Write=true
    Module.Index0Param2Write=true
    Module.Index0Param3Write=true
    Module.Index1Param0Write=true
    Module.Index1Param1Write=true
    Module.Index1Param2Write=true
    Module.Index1Param3Write=true
    Write Parameter Index 1=true
   )
   &CameraOffset(Camera Offset Amount=-17.9597)
   &self:CustomScaleMesh(ScaleBySpeed:Vector3f=<0.0002 0.0002 0>) {enabled=false, module=CustomScaleMesh}
  ]
  rendererClasses [&NiagaraMeshRendererProperties]
  renderers [
   {rendererClass=&NiagaraMeshRendererProperties, rendererIndex=0}
   {rendererClass=&NiagaraMeshRendererProperties, rendererIndex=1}
   {rendererClass=&NiagaraMeshRendererProperties, rendererIndex=2}
   {rendererClass=&NiagaraMeshRendererProperties, rendererIndex=3}
  ]
 }
 Rune {
  emitterUpdateScript [
   &EmitterState(
    Inactive Response=::NewEnumerator2
    Life Cycle Mode=::NewEnumerator1
    Loop Behavior=::NewEnumerator1
    Loop Duration Mode=::NewEnumerator1
    UseLoopDelay=false
   )
   &SpawnRate(SpawnRate=2, Use Spawn Probability=false) {enabled=false}
   &SpawnBurst_Instantaneous(Spawn Count=8, Use Spawn Probability=false, UseLoopCountLimit=false)
  ]
  particleSpawnScript [
   /SetVariables(Particles.SocketIndex=HlslExpression() {hlslExpression="fmod(Particles.UniqueID,2)"}) {isSetParameters=true, module=SetVariables_#3}
   &InitializeParticle(
    Color=<rgba 2 2 2 1>
    Lifetime=3
    Sprite Size Mode=::NewEnumerator3
    Uniform Sprite Size=55
    UsePositionOffset=false
    Write Material Random=true
   )
  ]
  particleUpdateScript [
   &SampleSkeletalMesh(
    AdditionalYawPitchRoll=false
    Bone / Socket Index:NiagaraInt32=$Particles.SocketIndex
    Bone Sampling Mode:ENiagaraBoneSamplingMode=::NewEnumerator4
    Position Offset Coordinate Space:ENiagaraCoordinateSpace=::Local
    Sampled Position Offset:Vector3f=<0 15 0>
    Skeletal Mesh:NiagaraDataInterfaceSkeletalMesh=$System.Skeletal Mesh Hand
   )
   /SetVariables(Particles.Position=$StackContext.SampleSkeletalMesh.SampledPosition) {isSetParameters=true, module=SetVariables_#5}
   &ParticleState(Kill Particles When Lifetime Has Elapsed=false, Let Infinitely Lived Particles Die When Emitter Deactivates=true)
   &ScaleSpriteSize(Uniform Curve Sprite Scale=DataInterface() {propertyValues {Curve {keys [{tangentMode=RCTM_User, time=0, value=0.75}, {tangentMode=RCTM_User, time=1, value=1}]}}}) {enabled=false}
   &self:RuneRotation(
    Intensity=DataInterface() {propertyValues {GpuSyncMode=SyncCpuToGpu, MaxElements=0, FloatData [0.3,0.3,0.6,0.6,1.2,1.2,0.85,0.85]}}
    MaxSpinDegree=360
    Size=DataInterface() {propertyValues {GpuSyncMode=SyncCpuToGpu, MaxElements=0, FloatData [1.5,1.5,1.2,1.2,0.2,0.2,0.6,0.6]}}
    Spin=DataInterface() {propertyValues {GpuSyncMode=SyncCpuToGpu, MaxElements=0, FloatData [1,1,3,3,0,0,0,0]}}
    SubuvIndex=DataInterface() {propertyValues {GpuSyncMode=SyncCpuToGpu, MaxElements=0, FloatData [1,1,0,0,2,2,3,3]}}
    Wiggle=1
   ) {module=RuneRotation}
   &ScaleColor(
    Color Value To Scale=$Particles.Color
    Scale Mode=::NewEnumerator1
    Scale RGBA=&MakeVector4_FromLinearColor(LinearColor=/Niagara/DynamicInputs/Lerp/SmoothLerpOverTime_Color(
      Smooth Value:LinearColor=<rgba 0.0388 0.4566 0.1249 2>
      Target Value:LinearColor=<rgba 0 10 0.6367 0.5>
      Convergence Rate:NiagaraFloat=&Abs_Float(Float:NiagaraFloat=&Multiply_Float(A=$Particles.SpinSpeed, B=50))
     ))
    ScaleRGBA=true
   )
   &CameraOffset(Camera Offset Amount=&RandomRangeFloat(Minimum=15, Maximum=25, Evaluation Type=::NewEnumerator1))
   &DynamicMaterialParameters(
    Index 0 Param 1=&Multiply_Float(A=$Particles.SpinSpeed, B=8)
    Module.Index0Param0Write=true
    Module.Index0Param1Write=true
    Module.Index0Param2Write=true
    Module.Index0Param3Write=true
   )
   &JitterPosition(Jitter Amount=1, Jitter Delay=0.1)
   &ScaleSpriteSize(
    Initial Sprite Size:Vector2f=$Particles.SpriteSize
    Scale Sprite Size Mode=::NewEnumerator0
    Uniform Scale Factor=&Add_Float(
     A=1
     B=&Multiply_Float(
      A=&VectorLength(Vector=$Engine.Owner.Velocity)
      B=&Add_Float(
       A=&Multiply_Float(A=0, B=&RandomRangeFloat(Minimum=-1, Maximum=1, Evaluation Type=::NewEnumerator1))
       B=0
      )
     )
    )
   ) {module=ScaleSpriteSize001}
  ]
  rendererClasses [&NiagaraSpriteRendererProperties]
  renderers [{rendererClass=&NiagaraSpriteRendererProperties}]
 }
}
summary {
 systemName=NS_SoilBlend
 emitters [
  {emitterName=Glow, rendererClasses [&NiagaraSpriteRendererProperties, &NiagaraLightRendererProperties]}
  {emitterName=Dots, rendererClasses [&NiagaraSpriteRendererProperties]}
  {emitterName=Rune, rendererClasses [&NiagaraSpriteRendererProperties]}
  {emitterName=HandDetail, rendererClasses [&NiagaraSpriteRendererProperties]}
  {emitterName=MeshRibbon, rendererClasses [&NiagaraMeshRendererProperties]}
 ]
 userVariables [
  User.bIsFrontEnd {type=NiagaraBool}
  User.bIsFrontEndPreview {type=NiagaraBool}
  User.ParaGlideLeanAlpha {type=NiagaraFloat, defaultValue=0.5, description=""}
 ]
}
